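# Unlit anime-style glass material for the Season 41 Reload Honor glider.
# Anime-parent maps support cel shading, Fresnel opacity and switchable Phong, Blinn, matcap, planar or cubemap highlights, alongside pre-skinned position masks.
# A masked Fresnel rim plus blended color drives emissive into MF_Season41_Reload_Honor_Glider, then the CharacterParent PostFX function.
# A camera-facing toon-highlight texture and a HueShift branch are built, but both bypass that main chain.
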Material /BRCosmetics/Gadgets/Assets/VinderTech_GliderChute/Glider_Season_41/Skins/Reload_Honor/Materials/M_Glider_Season41Reload_Honor_Glass
graph {
 properties {
  bAutomaticallySetUsageInEditor=false
  bTangentSpaceNormal=false
  bUseMaterialAttributes=true
  bUsedWithMeshDeformer=true
  bUsedWithMorphTargets=true
  bUsedWithSkeletalMesh=true
  previewMesh=/BRCosmetics/Gadgets/Assets/VinderTech_GliderChute/Glider_Season_41/Meshes/Season_41_Glider
  shadingModel=MSM_Unlit
  usedShadingModels=Unlit
 }
 nodes [
  n1 = NamedRerouteUsage()
  n2 = ScalarParameter() {defaultValue=100, parameterName=ToonHighlight_Scale}
  n3 = TextureSample(UVs=MaterialFunctionCall(UVs=NamedRerouteUsage(), Texture Scale=Multiply(A=@n1, B=@n2) {constB=10}) {materialFunction=&ScaleUVsByCenter}.UVs) {
   samplerSource=SSM_Clamp_WorldGroupSettings
   samplerType=SAMPLERTYPE_Masks
   texture=/BRCosmetics/Gadgets/Assets/VinderTech_GliderChute/Glider_GoggleBud/Textures/T_GoggleBud_Highlight
  }
  n4 = ScalarParameter() {defaultValue=3.5, parameterName=ToonHighlight_Intensity}
  n5 = TransformPosition(None=Constant3Vector() {constant=<rgba 0 0 1 0>}) {transformType=TRANSFORMPOSSOURCE_World}
  n6 = Saturate(None=DotProduct(A=CameraVectorWS(), B=Normalize(VectorInput=@n5)))
  n7 = Fresnel() {exponent=6}
  n8 = Power(Base=@n7) {constExponent=8}
  n9 = Saturate(None=MaterialFunctionCall!(Axis, Exponent, Normal) {materialFunction=/Game/Effects/Fort_Effects/Materials/_Functions/MF_AxisAlignedFrensnel}.Fresnel)
  n10 = Multiply(A=Power(Base=OneMinus(None=@n9)) {constExponent=16}) {constB=150}
  n11 = Multiply(A=@n7) {constB=400}
  n12 = LinearInterpolate(
   A=Saturate(None=Multiply(A=@n8) {constB=200})
   B=LinearInterpolate(A=Saturate(None=@n10), B=Saturate(None=@n11), Alpha=NamedRerouteUsage()) {constA=16, constB=4}
   Alpha=NamedRerouteUsage()
  )
  n13 = VectorParameter() {parameterName=FresnelOutlineColor, defaultValue=<rgba 0.3812 0.9168 1 1>}
  n14 = NamedRerouteUsage()
  n15 = Add(
   A=Multiply(A=Multiply(A=@n12, B=@n13.RGB)) {constB=0.5}
   B=LinearInterpolate(A=NamedRerouteUsage(), B=ComponentMask(None=@n14) {b=true, g=true, r=true}, Alpha=NamedRerouteUsage())
  )
  Add(A=Multiply(A=Multiply(A=@n3.R, B=@n4) {constB=3.5}, B=OneMinus(None=@n6)), B=@n15)
  Composite() {subgraphName=Construct Base Inputs}
  n16 = MaterialFunctionCall!(
   In_MaterialAttributes=SetMaterialAttributes!(MaterialAttributes, Emissive Color=@n15, Opacity=NamedRerouteUsage(), Normal=NamedRerouteUsage()) {attributeSetTypes [#1, #2, #3]}
   In_TechArtMask
  ) {materialFunction=&/MF_Season41_Reload_Honor_Glider}
  n17 = Constant()
  MaterialExpressionReroute_4 = Reroute(None=@n17)
  n18 = PreSkinnedPosition()
  n19 = ShaderStageSwitch(PixelShader=VertexInterpolator(VS=@n18).PS, VertexShader=@n18)
  MaterialExpressionReroute_5 = Reroute(None=@n19)
  n20 = MaterialFunctionCall!(Material) {materialFunction=/Game/Packages/Fortress_SharedMaterials/Base_Material_Functions/CharacterBrightnessandScale_Balance}
  MaterialExpressionReroute_6 = Reroute(None=@n20.Distance Blend)
  n21 = Fresnel() {baseReflectFraction=0, exponent=1}
  MaterialExpressionReroute_7 = Reroute(None=@n21)
  MaterialFunctionCall(
   Material=@n16.MaterialAttributes
   M=@MaterialExpressionReroute_4.M
   Pre Skinned Local Position=@MaterialExpressionReroute_5.Pre Skinned Local Position
   Distance Blend=@MaterialExpressionReroute_6.Distance Blend
   Fresnel=@MaterialExpressionReroute_7.Fresnel
  ) {materialFunction=&MF_CharacterParent_2023_PostFX}
  n22 = MaterialFunctionCall!(CustomUV) {materialFunction=/Game/Global/Materials/Character_Parent_2023/Anime_Parent_2023/MaterialFunctions/MFC_AnimeShader_2023_Pre}
  NamedRerouteDeclaration(None=@n22.Normal) {name=NormalMap, nodeColor=<rgba 0 0.0156 1 1>}
  n23 = NamedRerouteUsage()
  n24 = LinearInterpolate(
   A=NamedRerouteUsage()
   B=Constant3Vector() {constant=<rgba 0 0 1 0>}
   Alpha=ScalarParameter() {defaultValue=0.5, parameterName=SpecularHighlight_UpVectorBias}
  )
  n25 = Add(A=CameraVectorWS(), B=@n24)
  n26 = ScalarParameter() {group=02_Anime Parameters, parameterName=SpecularHighlight_LockedZVector, sliderMax=1}
  n27 = AppendVector(A=Constant2Vector() {g=1, r=1}, B=Saturate(None=OneMinus(None=@n26)))
  n28 = NamedRerouteDeclaration(None=Transform(None=NamedRerouteUsage())) {name=N_WS, nodeColor=<rgba 0.3125 0 1 1>}
  n29 = DotProduct(A=Normalize(VectorInput=Multiply(A=@n25, B=@n27)), B=@n28) {desc=Blinn}
  n30 = DotProduct(A=Normalize(VectorInput=Multiply(A=@n24, B=@n27)), B=ReflectionVectorWS(CustomWorldNormal=@n28)) {desc=Phong}
  n31 = Saturate(None=LinearInterpolate(
    A=StaticSwitchParameter(True=@n29, False=@n30) {desc=Default is the Phong Highlight, group=02_Anime Parameters, parameterName=UseBlinnHighlight, sortPriority=0}
    B=NamedRerouteUsage()
    Alpha=ScalarParameter() {group=02_Anime Parameters, parameterName=SpecularHighlight_CameraFacing_Blend, sliderMax=1}
   ))
  n32 = Constant() {r=1}
  n33 = ScalarParameter() {defaultValue=0.1, group=02_Anime Parameters, parameterName=SpecularHighlight_Size, sliderMax=1, sliderMin=0.0001}
  n34 = ScalarParameter() {defaultValue=0.5, group=02_Anime Parameters, parameterName=SpecularHighlight_Hardness, sliderMax=1}
  n35 = StaticSwitchParameter(
   True=NamedRerouteUsage()
   False=StaticSwitchParameter(True=NamedRerouteUsage(), False=SphereMask(A=@n31, B=@n32, Radius=@n33, Hardness=@n34) {hardnessPercent=78}) {group=02_Anime Parameters, parameterName=UsePlanarSpecular, sortPriority=1}
  ) {group=02_Anime Parameters, parameterName=UseCubeMapSpecular, sortPriority=1}
  NamedRerouteDeclaration(None=Multiply(
    A=StaticSwitchParameter(True=@n23, False=@n35) {group=02_Anime Parameters, parameterName=UseMatCapSpecular, sortPriority=1}
    B=ScalarParameter() {defaultValue=1, group=02_Anime Parameters, parameterName=SpecularHighlight_Intensity, sliderMax=8}
   )) {name=Specular_Curve, nodeColor=<rgba 1 0.4453 0 1>}
  n36 = LocalPosition() {localOrigin=InstancePreSkinning}
  n37 = NamedRerouteDeclaration(None=ShaderStageSwitch(PixelShader=VertexInterpolator(VS=@n36.XYZ).PS, VertexShader=@n36.XYZ)) {name=PreskinnedLocalPosition, nodeColor=<rgba 0.2891 0 1 1>}
  n38 = SphereMask(A=VectorParameter() {parameterName=SpaceShip_CenterPosition, defaultValue=<rgba 0 52 182 0>}.RGB, B=@n37) {attenuationRadius=80, hardnessPercent=0}
  n39 = NamedRerouteDeclaration(None=Saturate(None=Multiply(A=@n38) {constB=100})) {name=HelmetsMask, nodeColor=<rgba 0 0.0625 1 1>}
  n40 = Multiply(A=Add(A=ComponentMask(None=@n37) {b=true}) {constB=-240}) {constB=100}
  NamedRerouteDeclaration(None=Saturate(None=Multiply(A=@n39, B=@n40))) {name=PointyAreaMask, nodeColor=<rgba 0.4531 1 0 1>}
  n41 = TextureCoordinate()
  n42 = TextureSampleParameter2D!(
   UVs=MaterialFunctionCall!(Vector=ReflectionVectorWS()) {materialFunction=/Engine/Functions/Engine_MaterialFunctions01/Coordinates/LongLatToUV}.Result
   "DDX(UVs)"=DDX(Value=@n41)
   "DDY(UVs)"=DDY(Value=@n41)
   Apply View MipBias
  ) {
   group=01_Anime Maps
   mipValueMode=TMVM_Derivative
   parameterName=SpecularMap_Cube
   samplerSource=SSM_Wrap_WorldGroupSettings
   sortPriority=20
   texture=/Game/Global/Materials/Character_Parent_2023/Anime_Parent_2023/Textures/T_CubeMap_Highlight
  }
  NamedRerouteDeclaration(None=@n42.RGB) {name=CubeSpec, nodeColor=<rgba 1 0.4327 0 1>}
  n43 = Transform(None=Constant3Vector() {constant=<rgba 0 0 1 0>}) {transformSourceType=TRANSFORMSOURCE_View, transformType=TRANSFORM_Local}
  n44 = MaterialFunctionCall(Float2=MaterialFunctionCall(Vector or UVs=ComponentMask(None=@n43) {g=true, r=true}) {materialFunction=&VectorToRadialValue}.Radial Coordinates) {materialFunction=&BreakOutFloat2Components}
  n45 = AppendVector(A=Add(A=Multiply(A=@n44.R) {constB=4}) {constB=-1}, B=Multiply(A=ComponentMask(None=@n43) {b=true}) {constB=-1})
  n46 = WorldPosition()
  n47 = MaterialFunctionCall(Float3=Divide(
    A=TransformPosition(None=@n46.XYZ) {transformSourceType=TRANSFORMPOSSOURCE_World}
    B=ScalarParameter() {defaultValue=30, group=02_Anime Parameters, parameterName=SpecularHighlight_Planar_Scale, sliderMax=200}
   ) {constB=30}) {materialFunction=&BreakOutFloat3Components}
  n48 = MaterialFunctionCall(X=@n47.R, Y=@n47.R) {materialFunction=&MakeFloat2}
  n49 = MaterialFunctionCall(X=@n47.G, Y=@n47.G) {materialFunction=&MakeFloat2}
  n50 = Add(A=Multiply(A=StaticSwitchParameter(True=@n48.Result, False=@n49.Result) {defaultValue=true, group=02_Anime Parameters, parameterName=UsePlanarReflectionYPlane, sortPriority=2}) {constB=-1}) {constB=0.5}
  n51 = TextureSampleParameter2D!(
   UVs=Rotator(
    Coordinate=Add(A=@n45, B=@n50)
    Time=ScalarParameter() {group=02_Anime Parameters, parameterName=SpecularHighlight_Planar_Rotate, sliderMax=1, sliderMin=-1}
   ) {speed=3.1415}
   Apply View MipBias
  ) {
   group=01_Anime Maps
   parameterName=SpecularMap_Planar
   samplerSource=SSM_Wrap_WorldGroupSettings
   sortPriority=20
   texture=/Game/Global/Materials/Character_Parent_2023/Anime_Parent_2023/Textures/T_Planar_Highlight
  }
  NamedRerouteDeclaration(None=@n51.RGB) {name=PlanarSpec, nodeColor=<rgba 1 0.4327 0 1>}
  n52 = TransformPosition(None=NamedRerouteUsage()) {transformSourceType=TRANSFORMPOSSOURCE_World, transformType=TRANSFORMPOSSOURCE_View}
  NamedRerouteDeclaration(None=Divide(B=ComponentMask(None=@n52) {b=true})) {name=GlassPosition_ViewSpaceScale, nodeColor=<rgba 1 0.2714 0.0299 1>}
  n53 = GetMaterialAttributes(None=@n16.MaterialAttributes) {attributeGetTypes [#4, #1]}
  NamedRerouteDeclaration(None=@n53.Base Color) {name=In_BaseColor, nodeColor=<rgba 1 0.0234 0 1>}
  NamedRerouteDeclaration(None=@n53.Emissive Color) {name=In_EmissiveColor, nodeColor=<rgba 0 0.2969 1 1>}
  NamedRerouteDeclaration(None=MaterialFunctionCall(Hue Shift Percentage=ScalarParameter() {parameterName=HueShift}, Texture=NamedRerouteUsage()) {materialFunction=&HueShift}.Result) {name=Out_BaseColor, nodeColor=<rgba 0.6172 1 0 1>}
  n54 = TransformPosition(None=WorldPosition().XYZ) {transformSourceType=TRANSFORMPOSSOURCE_World, transformType=TRANSFORMPOSSOURCE_Camera}
  n55 = Normalize(VectorInput=VertexNormalWS())
  n56 = MaterialFunctionCall(XY=CrossProduct(A=Normalize(VectorInput=@n54), B=Transform(None=@n55) {transformSourceType=TRANSFORMSOURCE_World, transformType=TRANSFORM_Camera})) {materialFunction=&Swizzle}
  n57 = TextureSampleParameter2D(UVs=Add(A=Multiply(A=@n56.YX) {constB=0.5}) {constB=0.5}) {
   group=01_Anime Maps
   parameterName=SpecularMap_MatCap
   samplerSource=SSM_Wrap_WorldGroupSettings
   sortPriority=20
   texture=/Game/Global/Materials/Character_Parent_2023/Anime_Parent_2023/Textures/T_MatCap_Highlight
  }
  NamedRerouteDeclaration(None=@n57.RGB) {name=MatCapSpec, nodeColor=<rgba 1 0.4327 0 1>}
  NamedRerouteDeclaration(None=@n22.Specular_Color) {name=SpecularColor, nodeColor=<rgba 0.3359 1 0 1>}
  NamedRerouteDeclaration(None=MaterialFunctionCall(Hue Shift Percentage=ScalarParameter() {parameterName=HueShift}, Texture=NamedRerouteUsage()) {materialFunction=&HueShift}.Result) {name=Out_EmissiveColor, nodeColor=<rgba 0 1 0.7422 1>}
  NamedRerouteDeclaration(None=DotProduct(A=PixelNormalWS(), B=CameraVectorWS())) {name=NdotV, nodeColor=<rgba 1 0.6242 0 1>}
  NamedRerouteDeclaration(None=@n22.Color_Lit) {name=ColorLitMap, nodeColor=<rgba 1 0 0.6562 1>}
  n58 = MaterialFunctionCall() {materialFunction=&CameraDirectionVector}
  n59 = Multiply(A=Constant3Vector() {constant=<rgba 1 1 0 0>}, B=@n58.Result)
  n60 = Add(A=Multiply(B=Normalize(VectorInput=@n59)) {constA=0.25}, B=@n58.Result)
  n61 = NamedRerouteDeclaration(None=TransformPosition(None=VectorParameter() {group=Superlevel, parameterName=LocalPosition, sortPriority=1000, defaultValue=<rgba 0 0 50 0>}.RGB) {transformType=TRANSFORMPOSSOURCE_World}) {name=GlassPosition, nodeColor=<rgba 0 1 0.2734 1>}
  n62 = Constant() {r=45}
  n63 = Constant() {r=90}
  n64 = MaterialFunctionCall() {materialFunction=/Engine/Functions/Engine_MaterialFunctions02/Utility/FOV}
  n65 = Multiply(
   A=LinearInterpolate(A=Constant() {r=64}, B=Constant() {r=128}, Alpha=InverseLinearInterpolate(A=@n62, B=@n63, Value=@n64.FOV angle in degrees))
   B=Distance(A=@n61, B=CameraPositionWS())
  )
  NamedRerouteDeclaration(None=MaterialFunctionCall(Plane Normal Axis=Normalize(VectorInput=@n60), Plane Center=Add(A=@n58.Result, B=@n61), UV Scale=Multiply(A=@n65) {constB=0.0146}) {materialFunction=&VirtualPlaneCoordinates}.UVs) {name=GlassAllignedScreenPosition, nodeColor=<rgba 0.125 1 0 1>}
  NamedRerouteDeclaration(None=MaterialFunctionCall() {materialFunction=/Game/Global/Materials/Character_Parent_2023/Anime_Parent_2023/MaterialFunctions/MFC_AnimeShader_LightDirection}.Direction Vector) {name=LightVector, nodeColor=<rgba 1 0.934 0.5103 1>}
  NamedRerouteDeclaration(None=@n22.Color_Shaded) {name=ColorShadedMap, nodeColor=<rgba 0 0.9766 1 1>}
  n66 = Multiply(A=DotProduct(A=NamedRerouteUsage(), B=NamedRerouteUsage())) {constB=0.5}
  n67 = Saturate(None=Add(A=Add(A=@n66) {constB=0.5}, B=ScalarParameter() {group=02_Anime Parameters, parameterName=ShadingCurve_Bias, sliderMax=1, sliderMin=-1}))
  n68 = CurveAtlasRowParameter(CurveTime=@n67) {
   atlas=/Game/Characters/Shared/CurveAtlas/AthenaCharacterCurveAtlas
   curve=/Game/Characters/Shared/CurveAtlas/HB_TwoToneShadingCurve
   defaultValue=22
   group=02_Anime Parameters
   parameterName=ShadingCurve
   sortPriority=5
  }
  n69 = ScalarParameter() {defaultValue=0.02, group=02_Anime Parameters, parameterName=ShadingCurve_Width, sliderMax=1}
  n70 = ScalarParameter() {defaultValue=1, group=02_Anime Parameters, parameterName=ShadingCurve_Bands, sliderMax=4, sliderMin=1}
  n71 = ScalarParameter() {
   defaultValue=1
   desc="To convert the (mathmatical) linear greadient to linear perceived gradient"
   group=09_Anime Parameters_TechArt
   parameterName=ShadingCurve_Exp
   sliderMax=2.2
   sliderMin=0.4545
  }
  n72 = Power(
   Base=StaticSwitchParameter(
    True=ChannelMaskParameter(None=@n68) {group=02_Anime Parameters, parameterName=ShadingCurveChannel, sortPriority=10}.R
    False=MaterialFunctionCall!(NdL=@n67, Width=@n69, Bands Amount=@n70) {materialFunction=/Game/Global/Materials/Character_Parent_2023/Anime_Parent_2023/MaterialFunctions/MF_SmoothCelShadingWithBands}.Divide
   ) {group=02_Anime Parameters, parameterName=UseShadingCurve, sortPriority=-1}
   Exp=@n71
  ) {constExponent=2.2}
  NamedRerouteDeclaration(None=LinearInterpolate(A=NamedRerouteUsage(), B=NamedRerouteUsage(), Alpha=Saturate(None=@n72))) {name=ColorMixed, nodeColor=<rgba 0.0078 1 0 1>}
  n73 = ScalarParameter() {group=02_Anime Parameters, parameterName=Opacity_Min, sliderMax=1}
  n74 = ScalarParameter() {defaultValue=1, group=02_Anime Parameters, parameterName=Opacity_Max, sliderMax=1}
  n75 = Fresnel(
   ExponentIn=ScalarParameter() {defaultValue=3, group=02_Anime Parameters, parameterName=Opacity_FresnelExp, sliderMax=8}
   BaseReflectFractionIn=ScalarParameter() {defaultValue=0.1, group=02_Anime Parameters, parameterName=Opacity_FresnelBaseReflectionFrac, sliderMax=8}
  )
  NamedRerouteDeclaration(None=Add(A=NamedRerouteUsage(), B=LinearInterpolate(A=@n73, B=@n74, Alpha=@n75))) {name=Opacity, nodeColor=<rgba 1 0 0.3281 1>}
  PinBase!(M=@n17, Pre Skinned Local Position=@n19, Distance Blend=@n20.Distance Blend, Fresnel=@n21) {reroutePins [
    M {expression=&self:MaterialExpressionReroute_4}
    Pre Skinned Local Position {expression=&self:MaterialExpressionReroute_5}
    Distance Blend {expression=&self:MaterialExpressionReroute_6}
    Fresnel {expression=&self:MaterialExpressionReroute_7}
   ]}
  PinBase!()
  SetMaterialAttributes!(MaterialAttributes=@n53.MaterialAttributes, Base Color=NamedRerouteUsage(), Emissive Color=NamedRerouteUsage()) {attributeSetTypes [#4, #1]}
  TextureSample() {samplerType=SAMPLERTYPE_Masks, texture=&T_GoggleBud_Highlight}
 ]
}
